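# Control Rig for the F_MED_RinkHat_Trick face accessory that applies corrective offsets to hair bones as the head moves.
# A sequence first applies poses weighted by smoothed headUp_Target and HeadUp variables.
# It then layers poses weighted by the DeltaRotation function, which remaps head swing or twist relative to head_null into clamped weights.
# Additional twist-driven pose chains sit in the graph unconnected to execution.
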ControlRigBlueprint /BRCosmetics/Characters/Player/Female/Medium/Bodies/F_MED_RinkHat_Trick/Meshes/Parts/F_MED_RinkHat_Trick_FaceAcc_CtrlRig
graphs [
 New Function_ContainedGraph {nodes [
   n1 = RigVMFunctionEntryNode()
   RigVMFunction_MathBoolToInteger = &RigVMFunction_MathBoolToInteger(Value=@n1.onlyTwist)
   RigVMFunction_MathQuaternionSwingTwist = &RigVMFunction_MathQuaternionSwingTwist(
    twistAxis=@n1.twistAxis
    Input=&RigUnit_GetRelativeTransformForItem(bChildInitial=False, bParentInitial=False, child=@n1.child, parent=@n1.parent).RelativeTransform.Rotation
   )
   n2 = RigVMDispatchNode(
    Values=((),())
    Index=@RigVMFunction_MathBoolToInteger.Result
    Values.0=@RigVMFunction_MathQuaternionSwingTwist.Swing
    Values.1=@RigVMFunction_MathQuaternionSwingTwist.Twist
   ) {template="DISPATCH_RigVMDispatch_SelectInt32(in Index,in Values,out Result)"}
   RigVMFunction_MathQuaternionToEuler = &RigVMFunction_MathQuaternionToEuler(Value=@n2.Result, RotationOrder=@n1.Rotation Order)
   > DISPATCH_RigVMDispatch_Print = RigVMDispatchNode(ScreenColor=(R=1,G=1,B=1,A=1), ScreenDuration=1.5, Enabled=@n1.debugRotation, Value=@RigVMFunction_MathQuaternionToEuler.Result) {template="DISPATCH_RigVMDispatch_Print(in Prefix,in Value,in Enabled,in ScreenDuration,in ScreenColor)"}
   Cast = RigVMDispatchNode(Value=@n1.triggerAxis) {template="DISPATCH_RigVMDispatch_CastEnumToInt(in Value,out Result)"}
   DISPATCH_RigVMDispatch_SelectInt32 = RigVMDispatchNode(
    Values=(0,0,0)
    Index=@Cast.Result
    Values.0=@RigVMFunction_MathQuaternionToEuler.Result.X
    Values.1=@RigVMFunction_MathQuaternionToEuler.Result.Y
    Values.2=@RigVMFunction_MathQuaternionToEuler.Result.Z
   ) {template="DISPATCH_RigVMDispatch_SelectInt32(in Index,in Values,out Result)"}
   Remap = &RigVMFunction_MathDoubleRemap(
    bClamp=False
    Value=@DISPATCH_RigVMDispatch_SelectInt32.Result
    TargetMinimum=@n1.outputMinMax.X
    TargetMaximum=@n1.outputMinMax.Y
    SourceMinimum=@n1.rotationMinMax.X
    SourceMaximum=@n1.rotationMinMax.Y
   )
   Absolute = &RigVMFunction_MathDoubleAbs(Value=@Remap.Result)
   Multiply = &RigVMFunction_MathDoubleMul(
    A=&RigVMFunction_MathDoubleClamp(Value=RigVMDispatchNode(True=@Absolute.Result, Condition=@n1.absoluteValue, False=@Remap.Result) {template="DISPATCH_RigVMDispatch_If(in Condition,in True,in False,out Result)"}.Result).Result
    B=@n1.premult
   )
   > RigVMFunctionReturnNode(outputQuat=@n2.Result, outputWeight=@Multiply.Result)
  ]}
 RigVMFunctionLibrary {nodes [DeltaRotation v4 = RigVMCollapseNode(
    ExecuteContext=()
    Rotation Order=ZYX
    child=(Type=Bone,Name="None")
    outputMinMax=(X=0,Y=1)
    parent=(Type=Bone,Name="None")
    premult=1
    rotationMinMax=(X=0,Y=0)
    twistAxis=(X=1,Y=0,Z=0)
   )]}
 RigVMFunction_Sequence_1_ContainedGraph {nodes [
   n1 = RigVMFunctionEntryNode()
   RigVMFunction_Sequence = &RigVMFunction_Sequence(ExecuteContext=@n1.ExecuteContext)
   n2 = &RigVMFunction_Sequence(ExecuteContext=@RigVMFunction_Sequence.B)
   RigVMFunctionReturnNode(A=@RigVMFunction_Sequence.A, B=@n2.A, C=@n2.B)
  ]}
 RigVMModel {nodes [
   DeltaRotation v4_4 = RigVMFunctionReferenceNode(
    ExecuteContext=()
    Rotation Order=ZYX
    child=(Type=Bone,Name="head")
    onlyTwist=true
    outputMinMax=(X=0,Y=1)
    parent=(Type=Null,Name="head_null")
    premult=1
    rotationMinMax=(X=-30,Y=-80)
    triggerAxis=X
    twistAxis=(X=1,Y=0,Z=0)
   )
   > &RigUnit_ModifyTransforms(
    ItemToModify="((Item=(Type=Bone,Name=\"dyn_m_hair_01\"),Transform=(Rotation=(X=0.1486,Y=0.2532,Z=0.4798,W=0.8268),Translation=(X=0,Y=0,Z=0),Scale3D=(X=1,Y=1,Z=1))),(Item=(Type=Bone,Name=\"dyn_r_hair_01\"),Transform=(Rotation=(X=-0.1645,Y=-0.1652,Z=-0.4591,W=-0.8572),Transla ...(truncated)"
    WeightMinimum=0
    Weight=@"DeltaRotation v4_4".outputWeight
   )
   > DeltaRotation v4_5 = RigVMFunctionReferenceNode(
    Rotation Order=ZYX
    child=(Type=Bone,Name="head")
    onlyTwist=true
    outputMinMax=(X=0,Y=1)
    parent=(Type=Null,Name="head_null")
    premult=1
    rotationMinMax=(X=30,Y=80)
    triggerAxis=X
    twistAxis=(X=1,Y=0,Z=0)
   )
   > &RigUnit_ModifyTransforms(
    ItemToModify="((Item=(Type=Bone,Name=\"dyn_m_hair_01\"),Transform=(Rotation=(X=-0.0058,Y=0.186,Z=0.5002,W=-0.8457),Translation=(X=0,Y=0,Z=0),Scale3D=(X=1,Y=1,Z=1))),(Item=(Type=Bone,Name=\"dyn_r_hair_01\"),Transform=(Rotation=(X=0.073,Y=-0.0419,Z=0.3186,W=-0.9441),Translati ...(truncated)"
    WeightMinimum=0
    Weight=@"DeltaRotation v4_5".outputWeight
   )
   DeltaRotation v4_6 = RigVMFunctionReferenceNode(
    ExecuteContext=()
    Rotation Order=ZYX
    child=(Type=Bone,Name="head")
    onlyTwist=true
    outputMinMax=(X=0,Y=1)
    parent=(Type=Null,Name="head_null")
    premult=1
    rotationMinMax=(X=-5,Y=-80)
    triggerAxis=X
    twistAxis=(X=1,Y=0,Z=0)
   )
   > &RigUnit_ModifyTransforms(
    ItemToModify="((Item=(Type=Bone,Name=\"dyn_m_hair_01\"),Transform=(Rotation=(X=-0.1386,Y=-0.2588,Z=-0.447,W=-0.845),Translation=(X=0,Y=0,Z=0),Scale3D=(X=1,Y=1,Z=1))),(Item=(Type=Bone,Name=\"dyn_r_hair_01\"),Transform=(Rotation=(X=0.0488,Y=0.2279,Z=-0.0782,W=0.9693),Translat ...(truncated)"
    WeightMinimum=0
    Weight=@"DeltaRotation v4_6".outputWeight
   )
   > DeltaRotation v4_7 = RigVMFunctionReferenceNode(
    Rotation Order=ZYX
    child=(Type=Bone,Name="head")
    onlyTwist=true
    outputMinMax=(X=0,Y=1)
    parent=(Type=Null,Name="head_null")
    premult=1
    rotationMinMax=(X=5,Y=80)
    triggerAxis=X
    twistAxis=(X=1,Y=0,Z=0)
   )
   > &RigUnit_ModifyTransforms(
    ItemToModify="((Item=(Type=Bone,Name=\"dyn_m_hair_01\"),Transform=(Rotation=(X=-0.0058,Y=0.186,Z=0.5002,W=-0.8457),Translation=(X=0,Y=0,Z=0),Scale3D=(X=1,Y=1,Z=1))),(Item=(Type=Bone,Name=\"dyn_r_hair_01\"),Transform=(Rotation=(X=0.073,Y=-0.0419,Z=0.3186,W=-0.9441),Translati ...(truncated)"
    WeightMinimum=0
    Weight=@"DeltaRotation v4_7".outputWeight
   )
   RigUnit_BeginExecution = &RigUnit_BeginExecution()
   RigVMFunction_Sequence = RigVMAggregateNode(ExecuteContext=@RigUnit_BeginExecution.ExecutePin)
   AlphaInterp = &RigVMFunction_AlphaInterp(InterpSpeedDecreasing=10, InterpSpeedIncreasing=10, Scale=1, bClampResult=False, Value=RigVMVariableNode() {variable=headUp_Target}.Value)
   &RigUnit_ModifyTransforms(
    ItemToModify="((Item=(Type=Bone,Name=\"dyn_hair_01_l\"),Transform=(Rotation=(X=0,Y=0,Z=0,W=1),Translation=(X=0,Y=0,Z=0),Scale3D=(X=1,Y=1,Z=1))),(Item=(Type=Bone,Name=\"dyn_hair_02_l\"),Transform=(Rotation=(X=0,Y=-0.0654,Z=0,W=0.9979),Translation=(X=0,Y=0,Z=0),Scale3D=(X=1,Y ...(truncated)"
    WeightMinimum=0
    Weight=@AlphaInterp.Result
    ExecutePin=@RigVMFunction_Sequence.A
   )
   n1 = &RigVMFunction_AlphaInterp(InterpSpeedDecreasing=10, InterpSpeedIncreasing=10, Scale=1, bClampResult=False, Value=RigVMVariableNode() {variable=HeadUp}.Value)
   &RigUnit_ModifyTransforms(
    ItemToModify="((Item=(Type=Bone,Name=\"dyn_hair_01_l\"),Transform=(Rotation=(X=0,Y=0,Z=0,W=1),Translation=(X=0,Y=0,Z=0),Scale3D=(X=1,Y=1,Z=1))),(Item=(Type=Bone,Name=\"dyn_hair_02_l\"),Transform=(Rotation=(X=0,Y=0,Z=-0.061,W=0.9981),Translation=(X=0,Y=0,Z=0),Scale3D=(X=1,Y= ...(truncated)"
    WeightMinimum=0
    Weight=@n1.Result
    ExecutePin=@RigVMFunction_Sequence.B
   )
   &RigUnit_ModifyTransforms(
    ItemToModify=((Item=(Type=Bone,Name="head"),Transform=(Rotation=(X=0,Y=0,Z=0,W=1),Translation=(X=0,Y=0,Z=0),Scale3D=(X=1,Y=1,Z=1))))
    Weight=1
    WeightMinimum=0
    ExecutePin=@RigVMFunction_Sequence.C
   )
   > DeltaRotation v4_0 = RigVMFunctionReferenceNode(
    Rotation Order=ZYX
    child=(Type=Bone,Name="head")
    outputMinMax=(X=0,Y=1)
    parent=(Type=Null,Name="head_null")
    premult=1
    rotationMinMax=(X=-5,Y=-50)
    triggerAxis=Z
    twistAxis=(X=1,Y=0,Z=0)
   )
   > &RigUnit_ModifyTransforms(
    ItemToModify="((Item=(Type=Bone,Name=\"dyn_m_hair_01\"),Transform=(Rotation=(X=0,Y=0.2837,Z=0,W=-0.9589),Translation=(X=0,Y=0,Z=0),Scale3D=(X=1,Y=1,Z=1))),(Item=(Type=Bone,Name=\"dyn_r_hair_01\"),Transform=(Rotation=(X=-0.0737,Y=0.5021,Z=-0.1455,W=-0.8493),Translation=(X=0, ...(truncated)"
    WeightMinimum=0
    Weight=@"DeltaRotation v4_0".outputWeight
   )
   > DeltaRotation v4_2 = RigVMFunctionReferenceNode(
    Rotation Order=ZYX
    child=(Type=Bone,Name="head")
    outputMinMax=(X=0,Y=1)
    parent=(Type=Null,Name="head_null")
    premult=1
    rotationMinMax=(X=5,Y=50)
    triggerAxis=Z
    twistAxis=(X=1,Y=0,Z=0)
   )
   > &RigUnit_ModifyTransforms(
    ItemToModify=((Item=(Type=Bone,Name="dyn_base_back_down_hair_01"),Transform=(Rotation=(X=0,Y=0,Z=0,W=1),Translation=(X=0,Y=0,Z=10),Scale3D=(X=1,Y=1,Z=1))))
    WeightMinimum=0
    Weight=@"DeltaRotation v4_2".outputWeight
   )
   > DeltaRotation v4_1 = RigVMFunctionReferenceNode(
    Rotation Order=ZYX
    child=(Type=Bone,Name="head")
    onlyTwist=true
    outputMinMax=(X=0,Y=1)
    parent=(Type=Null,Name="head_null")
    premult=1
    rotationMinMax=(X=-5,Y=-80)
    triggerAxis=X
    twistAxis=(X=1,Y=0,Z=0)
   )
   > &RigUnit_ModifyTransforms(
    ItemToModify="((Item=(Type=Bone,Name=\"dyn_m_hair_01\"),Transform=(Rotation=(X=-0.1386,Y=-0.2588,Z=-0.447,W=-0.845),Translation=(X=0,Y=0,Z=0),Scale3D=(X=1,Y=1,Z=1))),(Item=(Type=Bone,Name=\"dyn_r_hair_01\"),Transform=(Rotation=(X=0.0488,Y=0.2279,Z=-0.0782,W=0.9693),Translat ...(truncated)"
    WeightMinimum=0
    Weight=@"DeltaRotation v4_1".outputWeight
   )
   > DeltaRotation v4_3 = RigVMFunctionReferenceNode(
    Rotation Order=ZYX
    child=(Type=Bone,Name="head")
    onlyTwist=true
    outputMinMax=(X=0,Y=1)
    parent=(Type=Null,Name="head_null")
    premult=1
    rotationMinMax=(X=5,Y=80)
    triggerAxis=X
    twistAxis=(X=1,Y=0,Z=0)
   )
   > &RigUnit_ModifyTransforms(
    ItemToModify="((Item=(Type=Bone,Name=\"dyn_m_hair_01\"),Transform=(Rotation=(X=-0.0058,Y=0.186,Z=0.5002,W=-0.8457),Translation=(X=0,Y=0,Z=0),Scale3D=(X=1,Y=1,Z=1))),(Item=(Type=Bone,Name=\"dyn_r_hair_01\"),Transform=(Rotation=(X=0.073,Y=-0.0419,Z=0.3186,W=-0.9441),Translati ...(truncated)"
    WeightMinimum=0
    Weight=@"DeltaRotation v4_3".outputWeight
   )
  ]}
]
hierarchy [
 root {loc=<0 0 0>, rot=<0 0 0 1>}
 pelvis {parent=root, loc=<0 -0.5301 98.0571>, rot=<0.0127 0.707 -0.0127 -0.707>}
 spine_01 {parent=pelvis, loc=<6.1208 -0.9328 0>, rot=<0 0 0.034 -0.9994>}
 spine_02 {parent=spine_01, loc=<6.5997 -0.2095 0>, rot=<0 0 0.0238 -0.9997>}
 spine_03 {parent=spine_02, loc=<6.4914 -0.1237 0>, rot=<0 0 0.0149 -0.9999>}
 spine_04 {parent=spine_03, loc=<6.9598 0.0794 0>, rot=<0 0 -0.0132 -0.9999>}
 spine_05 {parent=spine_04, loc=<6.9987 0 0>, rot=<0 0 -0.0015 -1>}
 neck_01 {parent=spine_05, loc=<11.3138 0.1529 0>, rot=<0 0 -0.2003 -0.9797>}
 neck_02 {parent=neck_01, loc=<4.5627 0 0>, rot=<0 0 0.0504 -0.9987>}
 head {parent=neck_02, loc=<5.3914 0 0>, rot=<0 0 0.115 -0.9934>}
 dyn_hair_base_l {parent=head, loc=<0.444 2.4182 -6.6328>, rot=<-0.0381 -0.9993 0 0>}
 dyn_hair_01_l {parent=dyn_hair_base_l, loc=<0 0 0>, rot=<0 0 0 1>}
 dyn_hair_02_l {parent=dyn_hair_01_l, loc=<5.5 0 0>, rot=<0 0 -0.0333 -0.9994>}
 dyn_hair_03_l {parent=dyn_hair_02_l, loc=<5.5 0 0>, rot=<0 0 -0.0333 -0.9994>}
 dyn_hair_04_l {parent=dyn_hair_03_l, loc=<5.5 0 0>, rot=<0 0 -0.0333 -0.9994>}
 dyn_hair_base_r {parent=head, loc=<0.5331 3.8418 6.254>, rot=<-0.0296 0.7065 0.7067 0.0237>}
 dyn_hair_01_r {parent=dyn_hair_base_r, loc=<0 0 0>, rot=<0 0 0 1>}
 dyn_hair_02_r {parent=dyn_hair_01_r, loc=<5.5 0 0>, rot=<-0.0007 0.0774 0.0093 -0.997>}
 dyn_hair_03_r {parent=dyn_hair_02_r, loc=<5.5 0 0>, rot=<-0.0007 0.0774 0.0093 -0.997>}
 dyn_hair_04_r {parent=dyn_hair_03_r, loc=<5.5 0 0>, rot=<0.0069 0.0771 -0.0889 -0.993>}
 dyn_hat_base {parent=head, loc=<15.7305 -3.9825 0.0028>, rot=<0.6702 -0.2215 0.2393 -0.6667>}
 dyn_hat_01 {parent=dyn_hat_base, loc=<0 0 0>, rot=<0 0 0 1>}
 dyn_m_hair_base {parent=head, loc=<1.0209 -7.2882 0>, rot=<0.1091 -0.6986 0.6986 -0.1091>}
 dyn_l_hair_base {parent=head, loc=<1.3068 -4.962 -4.9>, rot=<0.0854 -0.9107 0.3637 -0.1761>}
 dyn_r_hair_base {parent=head, loc=<1.3068 -4.962 4.9>, rot=<0.1615 -0.3925 0.9002 -0.0976>}
 dyn_r_side_hair_base {parent=head, loc=<0.2717 -0.7104 5.7>, rot=<0.1161 0.0729 0.9864 -0.0902>}
 dyn_l_side_hair_base {parent=head, loc=<0.2717 -0.7104 -5.7>, rot=<0.1318 -0.9851 -0.036 -0.1043>}
 dyn_l_side_hair_01 {parent=head, loc=<0.2717 -0.7104 -5.7>, rot=<0.1444 -0.9833 -0.0347 -0.1048>}
 dyn_l_side_hair_02 {parent=dyn_l_side_hair_01, loc=<5 0 0>, rot=<0 0 0.0128 -0.9999>}
 dyn_l_side_hair_03 {parent=dyn_l_side_hair_02, loc=<5 0 0>, rot=<0 0 0.0128 -0.9999>}
 dyn_l_side_hair_04 {parent=dyn_l_side_hair_03, loc=<5.0122 0 0>, rot=<0 0 0 1>}
 dyn_l_hair_01 {parent=head, loc=<1.3068 -4.962 -4.9>, rot=<0.0854 -0.9107 0.3637 -0.1761>}
 dyn_l_hair_02 {parent=dyn_l_hair_01, loc=<5 0 0>, rot=<0 0 0 1>}
 dyn_l_hair_03 {parent=dyn_l_hair_02, loc=<5 0 0>, rot=<0 0 0 1>}
 dyn_l_hair_04 {parent=dyn_l_hair_03, loc=<5.7872 0 0>, rot=<0 0 0 1>}
 dyn_m_hair_01 {parent=head, loc=<1.0209 -7.2882 0>, rot=<0.1091 -0.6986 0.6986 -0.1091>}
 dyn_m_hair_02 {parent=dyn_m_hair_01, loc=<5 0 0>, rot=<0 0 0 1>}
 dyn_m_hair_03 {parent=dyn_m_hair_02, loc=<5 0 0>, rot=<0 -0.0445 0 -0.999>}
 dyn_m_hair_04 {parent=dyn_m_hair_03, loc=<5 0 0>, rot=<0 -0.0445 0 -0.999>}
 dyn_r_hair_01 {parent=head, loc=<1.3068 -4.962 4.9>, rot=<0.1615 -0.3925 0.9002 -0.0976>}
 dyn_r_hair_02 {parent=dyn_r_hair_01, loc=<5 0 0>, rot=<0 0 0 1>}
 dyn_r_hair_03 {parent=dyn_r_hair_02, loc=<5 0 0>, rot=<0 0 0 1>}
 dyn_r_hair_04 {parent=dyn_r_hair_03, loc=<6.0236 0 0>, rot=<0 0 0 1>}
 dyn_r_side_hair_01 {parent=head, loc=<0.2717 -0.7104 5.7>, rot=<0.1161 0.0729 0.9864 -0.0902>}
 dyn_r_side_hair_02 {parent=dyn_r_side_hair_01, loc=<5 0 0>, rot=<0 0 -0.034 -0.9994>}
 dyn_r_side_hair_03 {parent=dyn_r_side_hair_02, loc=<5 0 0>, rot=<0 0 -0.034 -0.9994>}
 dyn_r_side_hair_04 {parent=dyn_r_side_hair_03, loc=<5.4999 0 0>, rot=<0 0 0 1>}
 dyn_l_front_hair_bade {parent=head, loc=<0.444 2.4182 -6.6328>, rot=<-0.0381 -0.9993 0 0>}
 dyn_l_front_hair_01 {parent=dyn_l_front_hair_bade, loc=<0 0 0>, rot=<0 0 0 1>}
 dyn_l_front_hair_02 {parent=dyn_l_front_hair_01, loc=<5.5 0 0>, rot=<0 0 -0.0333 -0.9994>}
 dyn_l_front_hair_03 {parent=dyn_l_front_hair_02, loc=<5.5 0 0>, rot=<0 0 -0.0333 -0.9994>}
 dyn_l_front_hair_04 {parent=dyn_l_front_hair_03, loc=<5.5 0 0>, rot=<0 0 -0.0333 -0.9994>}
 dyn_r_front_hair_bade {parent=head, loc=<0.5331 3.8418 6.254>, rot=<-0.0296 0.7065 0.7067 0.0237>}
 dyn_r_front_hair_01 {parent=dyn_r_front_hair_bade, loc=<0 0 0>, rot=<0 0 0 1>}
 dyn_r_front_hair_02 {parent=dyn_r_front_hair_01, loc=<5.5 0 0>, rot=<-0.0007 0.0774 0.0093 -0.997>}
 dyn_r_front_hair_03 {parent=dyn_r_front_hair_02, loc=<5.5 0 0>, rot=<-0.0007 0.0774 0.0093 -0.997>}
 dyn_r_front_hair_04 {parent=dyn_r_front_hair_03, loc=<5.5 0 0>, rot=<0.0069 0.0771 -0.0889 -0.993>}
 clavicle_l {parent=spine_05, loc=<5.4831 -0.1053 -3.2946>, rot=<0.0465 -0.7289 -0.009 -0.6829>}
 upperarm_l {parent=clavicle_l, loc=<9.8658 -1.2137 -0.4105>, rot=<0.0164 -0.3245 -0.0269 -0.9454>}
 lowerarm_l {parent=upperarm_l, loc=<26.465 -0.8345 0.259>, rot=<0.0036 0.0161 -0.2298 -0.9731>}
 hand_l {parent=lowerarm_l, loc=<22.7355 -0.3203 -0.6105>, rot=<-0.6784 -0.0223 0.0735 -0.7307>}
 thumb_01_l {parent=hand_l, loc=<1.3118 -0.3191 -2.459>, rot=<0.6546 -0.4042 0.0664 -0.6354>}
 thumb_02_l {parent=thumb_01_l, loc=<5.9336 -0.0448 0.0401>, rot=<-0.0037 0.0213 0.298 -0.9543>}
 thumb_03_l {parent=thumb_02_l, loc=<2.2223 0.3682 -0.0907>, rot=<0.0118 -0.013 0.0066 -0.9998>}
 index_metacarpal_l {parent=hand_l, loc=<1.6535 0.0183 -2.96>, rot=<0.0158 -0.0124 -0.0173 -0.9996>}
 index_01_l {parent=index_metacarpal_l, loc=<8.3008 0.0357 -0.0524>, rot=<-0.0044 0.0485 0.0924 -0.9945>}
 index_02_l {parent=index_01_l, loc=<3.3587 -0.3004 0.0452>, rot=<-0.002 0.0027 0.3225 -0.9466>}
 index_03_l {parent=index_02_l, loc=<2.5578 -0.123 -0.0252>, rot=<-0.002 0.0067 0.2558 -0.9667>}
 middle_metacarpal_l {parent=hand_l, loc=<2.1856 -0.1725 -0.5984>, rot=<0.006 0.0168 -0.0329 -0.9993>}
 middle_01_l {parent=middle_metacarpal_l, loc=<7.3275 -0.2447 0.0129>, rot=<-0.0208 0.0042 0.0858 -0.9961>}
 middle_02_l {parent=middle_01_l, loc=<4.0123 0.1401 0.1175>, rot=<-0.0218 0.0258 0.3512 -0.9357>}
 middle_03_l {parent=middle_02_l, loc=<2.6077 0.0534 0.1373>, rot=<0.0082 0.0188 0.1708 -0.9851>}
 ring_metacarpal_l {parent=hand_l, loc=<2.2074 -0.3625 1.3638>, rot=<-0.0284 0.0722 -0.0458 -0.9959>}
 ring_01_l {parent=ring_metacarpal_l, loc=<7.0295 -0.6197 -0.1446>, rot=<-0.0489 -0.0252 0.1892 -0.9804>}
 ring_02_l {parent=ring_01_l, loc=<3.6174 0.4647 -0.1027>, rot=<0 0 0.2806 -0.9598>}
 ring_03_l {parent=ring_02_l, loc=<2.4152 0.114 0.0199>, rot=<0 0 0.1189 -0.9929>}
 pinky_metacarpal_l {parent=hand_l, loc=<1.7616 -0.9753 3.3814>, rot=<-0.0397 0.0814 -0.069 -0.9935>}
 pinky_01_l {parent=pinky_metacarpal_l, loc=<7.1709 -0.4831 -0.1094>, rot=<-0.0654 -0.0035 0.2242 -0.9723>}
 pinky_02_l {parent=pinky_01_l, loc=<2.8274 0.0893 -0.1043>, rot=<0.0037 0.0011 0.2854 -0.9584>}
 pinky_03_l {parent=pinky_02_l, loc=<2.2046 0.0307 -0.0078>, rot=<-0.0002 0.0098 0.1445 -0.9895>}
 deform_wrist_dn_base_l {parent=hand_l, loc=<0 -1.5326 0>, rot=<0 0 0 -1>}
 deform_wrist_dn_l {parent=deform_wrist_dn_base_l, loc=<0 0 0>, rot=<0 0 0 1>}
 deform_wrist_up_base_l {parent=hand_l, loc=<0 2.3009 0>, rot=<0 0 0 -1>}
 deform_wrist_up_l {parent=deform_wrist_up_base_l, loc=<0 0 0>, rot=<0 0 0 1>}
 hand_attach_l {parent=hand_l, loc=<8.5 -3 0>, rot=<0.7071 0.7071 0 0>}
 weapon_l {parent=hand_l, loc=<-0.1974 -3.4111 0.3811>, rot=<0 0 0 1>}
 lowerarm_twist_02_l {parent=lowerarm_l, loc=<10.6707 0 0>, rot=<0 0 0 1>}
 lowerarm_twist_01_l {parent=lowerarm_l, loc=<18.316 0 0>, rot=<0 0 0 1>}
 deform_elbow_in_base_l {parent=lowerarm_l, loc=<0.1834 3.6062 -0.2572>, rot=<0 0 0 -1>}
 deform_elbow_in_l {parent=deform_elbow_in_base_l, loc=<0 0 0>, rot=<0 0 0 1>}
 deform_elbow_out_base_l {parent=lowerarm_l, loc=<0 -4 0>, rot=<0 0 0 1>}
 deform_elbow_out_l {parent=deform_elbow_out_base_l, loc=<0 0 0>, rot=<0 0 0 1>}
 dyn_l_sleeves_base {parent=lowerarm_l, loc=<2.0429 0.3733 -0.0076>, rot=<0.7026 -0.0109 0.0168 -0.7113>}
 dyn_l_bottam_sleeves_01 {parent=dyn_l_sleeves_base, loc=<-1.9616 7.5086 0.3655>, rot=<0 0 0 1>}
 dyn_l_bottam_sleeves_02 {parent=dyn_l_bottam_sleeves_01, loc=<0 0 0>, rot=<0 0 0 1>}
 dyn_l_front_sleeves_01 {parent=dyn_l_sleeves_base, loc=<-1.1215 1.8243 7.1985>, rot=<0 0 0 1>}
 dyn_l_front_sleeves_02 {parent=dyn_l_front_sleeves_01, loc=<0 0 0>, rot=<0 0 0 1>}
 dyn_l_top_sleeves_01 {parent=dyn_l_sleeves_base, loc=<-1.6879 -4.4071 1.3421>, rot=<0.0683 -0.0513 0.0035 -0.9963>}
 dyn_l_top_sleeves_02 {parent=dyn_l_top_sleeves_01, loc=<0 0 0>, rot=<0 0 0 1>}
 upperarm_twist_01_l {parent=upperarm_l, loc=<7.2135 -0.1134 0>, rot=<0 0 0 1>}
 deform_upperarm_bk_base_l {parent=upperarm_twist_01_l, loc=<-6.4053 -4.3746 0.314>, rot=<0 0 0 -1>}
 deform_upperarm_bk_l {parent=deform_upperarm_bk_base_l, loc=<0 0 0>, rot=<0 0 0 1>}
 deform_upperarm_fr_base_l {parent=upperarm_twist_01_l, loc=<-5.6312 5.092 0.3354>, rot=<0 0 0 -1>}
 deform_upperarm_fr_l {parent=deform_upperarm_fr_base_l, loc=<0 0 0>, rot=<0 0 0 1>}
 deform_upperarm_up_base_l {parent=upperarm_twist_01_l, loc=<-8.891 0.4825 3.9517>, rot=<0 0 0 -1>}
 deform_upperarm_up_l {parent=deform_upperarm_up_base_l, loc=<0 0 0>, rot=<0 0 0 1>}
 upperarm_twist_02_l {parent=upperarm_l, loc=<19.7998 -0.1134 0>, rot=<0 0 0 1>}
 clavicle_r {parent=spine_05, loc=<5.4831 -0.1053 3.2946>, rot=<0.7289 0.0465 -0.6829 0.009>}
 upperarm_r {parent=clavicle_r, loc=<-9.8658 1.2137 0.4103>, rot=<0.0164 -0.3245 -0.0269 -0.9454>}
 lowerarm_r {parent=upperarm_r, loc=<-26.465 0.8345 -0.259>, rot=<0.0036 0.0161 -0.2298 -0.9731>}
 hand_r {parent=lowerarm_r, loc=<-22.7355 0.3203 0.6106>, rot=<-0.6784 -0.0223 0.0735 -0.7307>}
 thumb_01_r {parent=hand_r, loc=<-1.3118 0.3191 2.459>, rot=<0.6546 -0.4042 0.0664 -0.6354>}
 thumb_02_r {parent=thumb_01_r, loc=<-5.9336 0.0448 -0.0401>, rot=<-0.0037 0.0213 0.298 -0.9543>}
 thumb_03_r {parent=thumb_02_r, loc=<-2.2223 -0.3682 0.0907>, rot=<0.0118 -0.013 0.0066 -0.9998>}
 index_metacarpal_r {parent=hand_r, loc=<-1.6535 -0.0183 2.96>, rot=<0.0158 -0.0124 -0.0173 -0.9996>}
 index_01_r {parent=index_metacarpal_r, loc=<-8.3008 -0.0357 0.0524>, rot=<-0.0044 0.0485 0.0924 -0.9945>}
 index_02_r {parent=index_01_r, loc=<-3.3587 0.3004 -0.0452>, rot=<-0.002 0.0027 0.3225 -0.9466>}
 index_03_r {parent=index_02_r, loc=<-2.5578 0.123 0.0252>, rot=<-0.002 0.0067 0.2558 -0.9667>}
 middle_metacarpal_r {parent=hand_r, loc=<-2.1856 0.1725 0.5984>, rot=<0.006 0.0168 -0.0329 -0.9993>}
 middle_01_r {parent=middle_metacarpal_r, loc=<-7.3275 0.2447 -0.0129>, rot=<-0.0208 0.0042 0.0858 -0.9961>}
 middle_02_r {parent=middle_01_r, loc=<-4.0123 -0.1401 -0.1175>, rot=<-0.0218 0.0258 0.3512 -0.9357>}
 middle_03_r {parent=middle_02_r, loc=<-2.6077 -0.0534 -0.1373>, rot=<0.0082 0.0188 0.1708 -0.9851>}
 ring_metacarpal_r {parent=hand_r, loc=<-2.2074 0.3625 -1.3638>, rot=<-0.0284 0.0722 -0.0458 -0.9959>}
 ring_01_r {parent=ring_metacarpal_r, loc=<-7.0295 0.6197 0.1446>, rot=<-0.0489 -0.0252 0.1892 -0.9804>}
 ring_02_r {parent=ring_01_r, loc=<-3.6174 -0.4647 0.1027>, rot=<0 0 0.2806 -0.9598>}
 ring_03_r {parent=ring_02_r, loc=<-2.4152 -0.114 -0.0199>, rot=<0 0 0.1189 -0.9929>}
 pinky_metacarpal_r {parent=hand_r, loc=<-1.7616 0.9753 -3.3814>, rot=<-0.0397 0.0814 -0.069 -0.9935>}
 pinky_01_r {parent=pinky_metacarpal_r, loc=<-7.1709 0.4831 0.1094>, rot=<-0.0654 -0.0035 0.2242 -0.9723>}
 pinky_02_r {parent=pinky_01_r, loc=<-2.8274 -0.0893 0.1043>, rot=<0.0037 0.0011 0.2854 -0.9584>}
 pinky_03_r {parent=pinky_02_r, loc=<-2.2046 -0.0307 0.0078>, rot=<-0.0002 0.0098 0.1445 -0.9895>}
 deform_wrist_dn_base_r {parent=hand_r, loc=<-0.0001 1.5328 0>, rot=<0 0 0 -1>}
 deform_wrist_dn_r {parent=deform_wrist_dn_base_r, loc=<0 0 0>, rot=<0 0 0 1>}
 deform_wrist_up_base_r {parent=hand_r, loc=<-0.0002 -2.3007 -0.0001>, rot=<0 0 0 -1>}
 deform_wrist_up_r {parent=deform_wrist_up_base_r, loc=<0 0 0>, rot=<0 0 0 1>}
 weapon_r {parent=hand_r, loc=<0.1974 3.4111 -0.3811>, rot=<0 0 0 1>}
 hand_attach_r {parent=hand_r, loc=<-8.5001 2.9999 0>, rot=<0 0 -0.7071 -0.7071>}
 lowerarm_twist_02_r {parent=lowerarm_r, loc=<-10.6707 0 0>, rot=<0 0 0 -1>}
 lowerarm_twist_01_r {parent=lowerarm_r, loc=<-18.316 0 0>, rot=<0 0 0 -1>}
 deform_elbow_in_base_r {parent=lowerarm_r, loc=<-0.1834 -3.6062 0.2572>, rot=<0 0 0 -1>}
 deform_elbow_in_r {parent=deform_elbow_in_base_r, loc=<0 0 0>, rot=<0 0 0 1>}
 deform_elbow_out_base_r {parent=lowerarm_r, loc=<0 4 -0.0001>, rot=<0 0 0 -1>}
 deform_elbow_out_r {parent=deform_elbow_out_base_r, loc=<0 0 0>, rot=<0 0 0 1>}
 dyn_r_sleeves_base {parent=lowerarm_r, loc=<-2.0428 -0.3733 0.0074>, rot=<-0.0168 -0.7113 0.7026 0.0109>}
 dyn_r_front_sleeves_01 {parent=dyn_r_sleeves_base, loc=<-1.1211 -1.8248 7.1983>, rot=<0 0 0 1>}
 dyn_r_front_sleeves_02 {parent=dyn_r_front_sleeves_01, loc=<0 0 0>, rot=<0 0 0 1>}
 dyn_r_bottam_sleeves_01 {parent=dyn_r_sleeves_base, loc=<-1.9614 -7.509 0.3654>, rot=<-0.7071 0 0 -0.7071>}
 dyn_r_bottam_sleeves_02 {parent=dyn_r_bottam_sleeves_01, loc=<0 0 0>, rot=<0 0 0 1>}
 dyn_r_top_sleeves_01 {parent=dyn_r_sleeves_base, loc=<-1.6881 4.4073 1.3422>, rot=<0.6562 -0.0338 -0.0387 -0.7528>}
 dyn_r_top_sleeves_02 {parent=dyn_r_top_sleeves_01, loc=<0 0 0>, rot=<0 0 0 1>}
 upperarm_twist_01_r {parent=upperarm_r, loc=<-7.2135 0.1134 0>, rot=<0 0 0 -1>}
 deform_upperarm_bk_base_r {parent=upperarm_twist_01_r, loc=<6.4053 4.3746 -0.314>, rot=<0 0 0 -1>}
 deform_upperarm_bk_r {parent=deform_upperarm_bk_base_r, loc=<0 0 0>, rot=<0 0 0 1>}
 deform_upperarm_fr_base_r {parent=upperarm_twist_01_r, loc=<5.6312 -5.092 -0.3354>, rot=<0 0 0 -1>}
 deform_upperarm_fr_r {parent=deform_upperarm_fr_base_r, loc=<0 0 0>, rot=<0 0 0 1>}
 deform_upperarm_up_base_r {parent=upperarm_twist_01_r, loc=<8.891 -0.4825 -3.9517>, rot=<0 0 0 -1>}
 deform_upperarm_up_r {parent=deform_upperarm_up_base_r, loc=<0 0 0>, rot=<0 0 0 1>}
 upperarm_twist_02_r {parent=upperarm_r, loc=<-19.7998 0.1134 0>, rot=<0 0 0 -1>}
 head_null {parent=spine_05, type=Null, loc=<20.6573 3.5489 0>, rot=<0 0 -0.036 -0.9994>}
 dyn_hood_base {parent=spine_05, loc=<9.6045 -11.6104 0>, rot=<0.5417 -0.4545 0.4545 -0.5417>}
 dyn_hood {parent=dyn_hood_base, loc=<0 0 0>, rot=<0 0 0 1>}
 attach_backpack {parent=spine_05, loc=<-0.123 -13.473 0>, rot=<-0.0093 -0.707 0.0093 -0.707>}
 Backpack_BR {parent=attach_backpack, type=Null, loc=<0 0 0>, rot=<0 0 0 1>}
 attach_cape {parent=spine_05, loc=<9.017 -11.714 0>, rot=<0.0616 -0.7044 -0.0616 -0.7044>}
 cape {parent=attach_cape, type=Null, loc=<0 0 0>, rot=<0 0 0 1>}
 dyn_base_back_down_hair {parent=spine_05, loc=<9.1562 -9.068 0>, rot=<0.5934 -0.3845 0.3845 -0.5934>}
 dyn_base_back_down_hair_01 {parent=dyn_base_back_down_hair, loc=<0 0 0>, rot=<0 0 0 1>}
 deform_pec_aim_base_l {parent=spine_04, loc=<3.2456 9.3309 -1.2237>, rot=<0.0603 -0.4033 0.3318 -0.8507>}
 deform_pec_aim_l {parent=deform_pec_aim_base_l, loc=<0 0 0>, rot=<0 0 0 1>}
 deform_pec_base_l {parent=deform_pec_aim_l, loc=<9.3665 0 0>, rot=<0 0 0 1>}
 deform_pec_l {parent=deform_pec_base_l, loc=<0 0 0>, rot=<0 0 0 1>}
 deform_pec_aim_base_r {parent=spine_04, loc=<3.2452 9.3309 1.2237>, rot=<-0.4033 -0.0603 0.8507 0.3318>}
 deform_pec_aim_r {parent=deform_pec_aim_base_r, loc=<0 0 0>, rot=<0 0 0 1>}
 deform_pec_base_r {parent=deform_pec_aim_r, loc=<-9.3667 -0.0003 -0.0004>, rot=<0 0 0 1>}
 deform_pec_r {parent=deform_pec_base_r, loc=<0 0 0>, rot=<0 0 0 1>}
 dyn_front_jacket_base {parent=spine_03, loc=<1.1709 12.6095 0>, rot=<-0.0387 0.706 0.706 -0.0387>}
 dyn_front_jacket_01 {parent=dyn_front_jacket_base, loc=<0 0 0>, rot=<0 0 0 1>}
 dyn_front_jacket_02 {parent=dyn_front_jacket_01, loc=<3.4871 0 0>, rot=<0 0 0 1>}
 dyn_l_front_side_jacket_base {parent=spine_03, loc=<1.2165 10.0006 -9.5392>, rot=<0.0642 -0.886 -0.4585 0.0242>}
 dyn_l_front_side_jacket_01 {parent=dyn_l_front_side_jacket_base, loc=<0 0 0>, rot=<0 0 0 1>}
 dyn_l_front_side_jacket_02 {parent=dyn_l_front_side_jacket_01, loc=<3.4871 0 0>, rot=<0 0 0 1>}
 dyn_l_left_jacket_base {parent=spine_03, loc=<1.3934 0.9188 -12.6229>, rot=<0.0547 -0.9985 -0.0004 -0.0065>}
 dyn_l_left_jacket_01 {parent=dyn_l_left_jacket_base, loc=<0 0 0>, rot=<0 0 0 1>}
 dyn_l_left_jacket_02 {parent=dyn_l_left_jacket_01, loc=<3.4871 0 0>, rot=<0 0 0 1>}
 dyn_back_jacket_base {parent=spine_03, loc=<1.5018 -8.3619 0>, rot=<0.036 -0.7062 0.7062 -0.036>}
 dyn_back_jacket_01 {parent=dyn_back_jacket_base, loc=<0 0 0>, rot=<0 0 0 1>}
 dyn_back_jacket_02 {parent=dyn_back_jacket_01, loc=<3.4871 0 0>, rot=<0 0 0 1>}
 dyn_l_back_jacket_base {parent=spine_03, loc=<1.476 -6.0244 -9.528>, rot=<0.0436 -0.9184 0.3926 -0.0214>}
 dyn_l_back_jacket_01 {parent=dyn_l_back_jacket_base, loc=<0 0 0>, rot=<0 0 0 1>}
 dyn_l_back_jacket_02 {parent=dyn_l_back_jacket_01, loc=<3.4871 0 0>, rot=<0 0 0 1>}
 dyn_r_back_jacket_base {parent=spine_03, loc=<1.4756 -6.0244 9.528>, rot=<0.0274 -0.5189 0.8535 -0.0401>}
 dyn_r_back_jacket_01 {parent=dyn_r_back_jacket_base, loc=<0 0 0>, rot=<0 0 0 1>}
 dyn_r_back_jacket_02 {parent=dyn_r_back_jacket_01, loc=<3.4863 0 0>, rot=<0 0 0 1>}
 dyn_r_left_jacket_base {parent=spine_03, loc=<1.3931 0.9188 12.6229>, rot=<0.0065 0.0004 0.9985 -0.0547>}
 dyn_r_left_jacket_01 {parent=dyn_r_left_jacket_base, loc=<0 0 0>, rot=<0 0 0 1>}
 dyn_r_left_jacket_02 {parent=dyn_r_left_jacket_01, loc=<3.4873 0 0>, rot=<0 0 0 1>}
 dyn_r_front_side_jacket_base {parent=spine_03, loc=<1.2164 10.0006 9.5392>, rot=<-0.0236 0.4504 0.8902 -0.0644>}
 dyn_r_front_side_jacket_01 {parent=dyn_r_front_side_jacket_base, loc=<0 0 0>, rot=<0 0 0 1>}
 dyn_r_front_side_jacket_02 {parent=dyn_r_front_side_jacket_01, loc=<3.4867 0 0>, rot=<0 0 0 1>}
 thigh_l {parent=pelvis, loc=<-1.0855 -0.2109 -8.7053>, rot=<0.0009 0.0355 0.0155 -0.9993>}
 calf_l {parent=thigh_l, loc=<-42.5284 -0.0019 -0.7297>, rot=<0.0006 -0.0161 -0.0391 -0.9991>}
 foot_l {parent=calf_l, loc=<-46.7401 -2.5014 -0.592>, rot=<0.0427 -0.0238 0.0353 -0.9982>}
 ball_l {parent=foot_l, loc=<-3.5658 10.8066 -0.1544>, rot=<0.0001 -0.0001 -0.7063 -0.7079>}
 calf_twist_02_l {parent=calf_l, loc=<-19.1546 -1.7943 -0.2638>, rot=<0.0237 0.0095 -0.0035 -0.9997>}
 dyn_l_pant_base {parent=calf_twist_02_l, loc=<-4.53 0 0>, rot=<0 0.7071 0.7071 0>}
 dyn_l_pant {parent=dyn_l_pant_base, loc=<0 0 0>, rot=<0 0 0 1>}
 calf_twist_01_l {parent=calf_l, loc=<-35.8483 -1.9197 -0.5794>, rot=<0.0237 0.0095 -0.0035 -0.9997>}
 dyn_l_skate_tongue_base {parent=calf_twist_01_l, loc=<2.112 4.5475 -0.613>, rot=<-0.6483 -0.2937 -0.2425 -0.6592>}
 dyn_l_skate_tongue_01 {parent=dyn_l_skate_tongue_base, loc=<0 0 0>, rot=<0 0 0 1>}
 dyn_l_skate_tongue_02 {parent=dyn_l_skate_tongue_01, loc=<1.554 0 0>, rot=<0.0314 -0.685 0.0155 -0.7277>}
 dyn_l_skate_tongue_03 {parent=dyn_l_skate_tongue_02, loc=<3.3636 0 0>, rot=<0 -0.2005 0 -0.9797>}
 deform_knee_in_base_l {parent=calf_l, loc=<0.8031 -5.0231 0.2456>, rot=<0 0 0 -1>}
 deform_knee_in_l {parent=deform_knee_in_base_l, loc=<0 0 0>, rot=<0 0 0 1>}
 deform_knee_out_base_l {parent=calf_l, loc=<0 4 0>, rot=<0 0 0 -1>}
 deform_knee_out_l {parent=deform_knee_out_base_l, loc=<0 0 0>, rot=<0 0 0 1>}
 thigh_twist_01_l {parent=thigh_l, loc=<-27.952 0 0>, rot=<0.0013 0 0 -1>}
 deform_groin_base_l {parent=thigh_l, loc=<0.4801 5.6776 -0.2486>, rot=<0 0 0 -1>}
 deform_groin_l {parent=deform_groin_base_l, loc=<0 0 0>, rot=<0 0 0 1>}
 thigh_r {parent=pelvis, loc=<-1.0855 -0.2109 8.7053>, rot=<0.0355 -0.0009 0.9993 0.0155>}
 calf_r {parent=thigh_r, loc=<42.5284 0.0019 0.7297>, rot=<0.0006 -0.0161 -0.0391 -0.9991>}
 foot_r {parent=calf_r, loc=<46.7401 2.5014 0.592>, rot=<0.0427 -0.0238 0.0353 -0.9982>}
 ball_r {parent=foot_r, loc=<3.5658 -10.8066 0.1544>, rot=<0.0001 -0.0001 -0.7063 -0.7079>}
 calf_twist_02_r {parent=calf_r, loc=<19.1546 1.7943 0.2638>, rot=<0.0237 0.0095 -0.0035 -0.9997>}
 dyn_r_pant_base {parent=calf_twist_02_r, loc=<4.53 0 0>, rot=<-0.7071 0 0 -0.7071>}
 dyn_r_pant {parent=dyn_r_pant_base, loc=<0 0 0>, rot=<0 0 0 1>}
 calf_twist_01_r {parent=calf_r, loc=<35.8483 1.9197 0.5794>, rot=<0.0237 0.0095 -0.0035 -0.9997>}
 dyn_r_skate_tongue_base {parent=calf_twist_01_r, loc=<-1.4251 -4.497 0.5288>, rot=<0.3968 -0.5917 -0.5067 0.4855>}
 dyn_r_skate_tongue_01 {parent=dyn_r_skate_tongue_base, loc=<0 0 0>, rot=<0 -0.0503 0 -0.9987>}
 dyn_r_skate_tongue_02 {parent=dyn_r_skate_tongue_01, loc=<1.554 0 0>, rot=<0.0343 -0.4719 0.0068 -0.8809>}
 dyn_r_skate_tongue_03 {parent=dyn_r_skate_tongue_02, loc=<3.3636 0 0>, rot=<0 -0.0587 0 -0.9983>}
 deform_knee_in_base_r {parent=calf_r, loc=<-0.8031 5.0231 -0.2457>, rot=<0 0 0 -1>}
 deform_knee_in_r {parent=deform_knee_in_base_r, loc=<0 0 0>, rot=<0 0 0 1>}
 deform_knee_out_base_r {parent=calf_r, loc=<0 -4 0>, rot=<0 0 0 -1>}
 deform_knee_out_r {parent=deform_knee_out_base_r, loc=<0 0 0>, rot=<0 0 0 1>}
 thigh_twist_01_r {parent=thigh_r, loc=<27.952 0 0>, rot=<0.0013 0 0 -1>}
 deform_groin_base_r {parent=thigh_r, loc=<-0.4801 -5.6776 0.2486>, rot=<0 0 0 -1>}
 deform_groin_r {parent=deform_groin_base_r, loc=<0 0 0>, rot=<0 0 0 1>}
 deform_glute_base_l {parent=pelvis, loc=<-7.3466 -9.0213 -9.1245>, rot=<0.0172 0.031 -0.4477 -0.8935>}
 deform_glute_l {parent=deform_glute_base_l, loc=<0 0 0>, rot=<0 0 0 1>}
 deform_glute_base_r {parent=pelvis, loc=<-7.3464 -9.0213 9.1245>, rot=<0.031 -0.0172 0.8935 -0.4477>}
 deform_glute_r {parent=deform_glute_base_r, loc=<0 0 0>, rot=<0 0 0 1>}
]
variables [
 HeadDown {type=bool, default=False}
 HeadUp {type=double, default=0}
 HeadUp_New {type=double, default=0}
 headUp_Target {type=double, default=0}
]
