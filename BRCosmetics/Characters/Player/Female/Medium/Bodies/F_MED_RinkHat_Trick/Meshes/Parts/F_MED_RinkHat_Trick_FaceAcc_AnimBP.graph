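# Post-process-style secondary-motion AnimBP for the RinkHat Trick cosmetic's face accessory, following the attached parent's pose.
# Head-relative RigidBody simulation, switching physics assets for front end, feeds spring Anim Dynamics on the hat bone, then a Control Rig.
# Updates derive head-up values for the rig, gating one by targeting.
# Computed head-down, head-turn, gravity override and RBAN settings variables are unread in these graphs.
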AnimBlueprint /BRCosmetics/Characters/Player/Female/Medium/Bodies/F_MED_RinkHat_Trick/Meshes/Parts/F_MED_RinkHat_Trick_FaceAcc_AnimBP
target_skeleton=&/F_MED_RinkHat_Trick_FaceAcc_Skeleton
graphs [
 AnimGraph {nodes [
   n1 = RigidBody!(ComponentPose=Local To Component(LocalPose=Use cached pose 'Input Pose'()), Alpha, OverrideWorldGravity:Vector="0,0,0") -> (Pose) {
    bOverrideWorldGravity=true
    overridePhysicsAsset=&/F_MED_RinkHat_Trick_FaceAcc_Physics
    simulationSpace=BaseBoneSpace
    alphaBoolBlend {blendOption=Linear}
    baseBoneRef {boneName=head}
    overrideWorldGravity {z=-980}
    simSpaceSettings {maxAngularVelocity=3, maxLinearAcceleration=300, maxLinearVelocity=500, velocityScaleZ=0.3, worldAlpha=0.5}
   }
   n2 = Component To Local(ComponentPose=@n1)
   n3 = RigidBody(ComponentPose=Local To Component(LocalPose=Use cached pose 'Input Pose'()), OverrideWorldGravity="0,0,0") {
    bOverrideWorldGravity=true
    overridePhysicsAsset=&/F_MED_RinkHat_Trick_FaceAcc_FE_Physics
    simulationSpace=BaseBoneSpace
    alphaBoolBlend {blendOption=Linear}
    baseBoneRef {boneName=head}
    overrideWorldGravity {z=-980}
    simSpaceSettings {maxAngularVelocity=3, maxLinearAcceleration=300, maxLinearVelocity=500, velocityScaleZ=0.3, worldAlpha=0.5}
   }
   n4 = Component To Local(ComponentPose=@n3)
   n5 = Get bIsInFrontEnd()
   n6 = Anim Dynamics - Bone: dyn_hat_01!(
    ComponentPose=Local To Component(LocalPose=Blend Poses by bool(BlendPose_0=@n4, BlendPose_1=@n2, bActiveValue=@n5))
    Alpha=1
    GravityOverride:Vector="0,0,-980"
    ExternalForce="0,0,0"
   ) -> (Pose) {
    angularDampingOverride=1
    angularSpringConstant=20000
    bAngularSpring=true
    bOverrideAngularDamping=true
    bUseGravityOverride=true
    gravityScale=0.35
    simulationSpace=RootRelative
    boundBone {boneName=dyn_hat_01}
    componentLinearAccScale=<1 1 1>
    componentLinearVelScale=<1 1 1>
    gravityOverride {z=-980}
    physicsBodyDefinitions [{
      boundBone {boneName=dyn_hat_01}
      boxExtents=<20 30 30>
      constraintSetup {angularLimitsMax=<0 10 10>, angularLimitsMin=<0 -10 -10>, angularTarget=<1 0 0>}
      localJointOffset=<4 0 0>
     }]
   }
   n7 = Control Rig!(
    Source=Component To Local(ComponentPose=@n6)
    Alpha
    headUp_Target:double=Get HeadUP_Target() -> (HeadUP_Target:double)
    HeadUp:double=Get Head Up _New() -> (Head Up _New:double)
   ) -> (Pose) {controlRigAssetReference {blueprintRigClass=&/F_MED_RinkHat_Trick_FaceAcc_CtrlRig.F_MED_RinkHat_Trick_FaceAcc_CtrlRig_C}}
   Output Pose(Result=@n7) {layerGroup=None}
   Save cached pose 'Input Pose'(Pose=Copy Pose From Mesh() {bUseAttachedParent=true}) {cacheName=Input Pose}
  ]}
 EventGraph {nodes [
   Event BlueprintUpdateAnimation()
   > n1 = Sequence!(execute) -> (then_0, then_1, then_2:Exec)
   > n2 = "Gravity Override State Function (Parameter Asset) (Nativized)"(
    CharacterPartAnimInstance=Self-Reference()
    GameTimeInSeconds=GetGameTimeInSeconds()
    Parameters=&/F_MED_RinkHat_Trick_FaceAcc_Parameters
   )
   > Set GravityOverride(GravityOverride=@n2)
   > Set RBAN Settings Out(RBAN Settings Out=@n2.OutRigidBodyAnimNodeParameters)
   n3 = BreakTransform(InTransform=GetDeltaTransformFromRefPose(self=GetOwnerSkeletalMeshComponent(), BoneName=head, BaseName=spine_05))
   Set Head Up _New(execute=@n1.then_1, Head Up _New:double=MapRangeClamped(Value=BreakRotator(InRot=@n3.Rotation).Yaw, InRangeA=0, InRangeB=-75)) -> (Output_Get:double)
   > Set HeadUP_Target(HeadUP_Target:double=SelectFloat(A=Get Head Up _New(), B=0, bPickA=Get bIsTargeting())) -> (Output_Get:double)
   n4 = GetOwnerSkeletalMeshComponent()
   Set Head Up _New(
    execute=@n1.then_2
    Head Up _New=GetAlphaFromReferenceTransformDelta(SkeletalMeshComponent=@n4, BoneName=spine_05, BaseName=head, InputRange=(X=10,Y=70), TwistAxis="0,0,1")
   )
   > Set Head Down(Head Down:double=GetAlphaFromReferenceTransformDelta(SkeletalMeshComponent=@n4, BoneName=spine_05, BaseName="head ", DeltaComponentType=EulerY, InputRange=(X=-4,Y=-9), TwistAxis="0,0,1")) -> (Output_Get:double)
   > Set Head_Turn_L(Head_Turn_L:double=GetAlphaFromReferenceTransformDelta(SkeletalMeshComponent=@n4, BoneName=spine_05, BaseName=head, DeltaComponentType=EulerX, InputRange=(X=0,Y=-22), TwistAxis="1,0,0")) -> (Output_Get:double)
   > Set Head_Turn_R(Head_Turn_R:double=GetAlphaFromReferenceTransformDelta(SkeletalMeshComponent=@n4, BoneName=spine_05, BaseName=head, DeltaComponentType=EulerX, InputRange=(X=0,Y=22), TwistAxis="1,0,0")) -> (Output_Get:double)
  ]}
]
variables [
 GravityOverride {type=Vector}
 Head Down {type=number}
 Head Up _New {type=number}
 Head_Turn_L {type=number}
 Head_Turn_R {type=number}
 HeadUP_Target {type=number}
 RBAN Settings Out {type=FortRigidBodyAnimNodeParameters}
 Skel Mesh {type=SkeletalMeshComponent}
]
