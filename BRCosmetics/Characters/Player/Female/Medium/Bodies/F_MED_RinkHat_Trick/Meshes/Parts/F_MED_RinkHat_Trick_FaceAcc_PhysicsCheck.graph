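# Physics asset for the face accessory part of the female medium RinkHat Trick character, set up for left and right dynamic hair bone chains.
# Capsule bodies on the upper spine, clavicles, upper arms, neck and head give the hair a rough body shape to collide with.
# Each hair chain runs from a small base body on the head through four segments with narrow swing limits and zero twist range.
# Most body links and both hair base joints rotate freely.
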PhysicsAsset /BRCosmetics/Characters/Player/Female/Medium/Bodies/F_MED_RinkHat_Trick/Meshes/Parts/F_MED_RinkHat_Trick_FaceAcc_PhysicsCheck
bodies {
 clavicle_l [
  {shapeType=Capsule, length=5.0044, radius=6.3105, shapeName=clavicle_l_capsule, center=<3.4479 -0.7062 -1.8774>, rotation=<rot 90 0 0>}
  {
   shapeType=Capsule
   length=10.923
   radius=2.3478
   shapeName=clavicle_l_capsule
   center=<2.4031 1.52 4.1613>
   rotation=<rot -71.0816 -41.5997 -4.3534>
  }
  {shapeType=Capsule, length=5, radius=5, shapeName=clavicle_l_capsule, center=<11.387 0 5>, rotation=<rot 90 0 0>}
 ]
 clavicle_r [
  {shapeType=Capsule, length=4.6445, radius=6.6704, shapeName=clavicle_r_capsule, center=<-3.4479 0.7062 1.8774>, rotation=<rot 90 180 0>}
  {
   shapeType=Capsule
   length=10.9235
   radius=2.3473
   shapeName=clavicle_r_capsule
   center=<-2.4031 -1.52 -4.1613>
   rotation=<rot -71.0816 -41.5997 175.6466>
  }
 ]
 dyn_hair_01_l [{shapeType=Capsule, length=3.5767, radius=3, shapeName=dyn_hair_01_l_capsule, center=<4 0 0>, rotation=<rot 90 0 0>}]
 dyn_hair_01_r [{shapeType=Capsule, length=3.6542, radius=3, shapeName=dyn_hair_01_r_capsule, center=<4 0 0>, rotation=<rot 90 0 0>}]
 dyn_hair_02_l [{shapeType=Capsule, length=3.1505, radius=3, shapeName=dyn_hair_02_l_capsule, center=<4 -0.5 0>, rotation=<rot 90 0 0>}]
 dyn_hair_02_r [{shapeType=Capsule, length=3.5, radius=2, shapeName=dyn_hair_02_r_capsule, center=<4 0 -0.5>, rotation=<rot 90 0 0>}]
 dyn_hair_03_l [{shapeType=Capsule, length=3.0622, radius=2, shapeName=dyn_hair_03_l_capsule, center=<4 0 0>, rotation=<rot 90 9.1949 0>}]
 dyn_hair_03_r [{shapeType=Capsule, length=3.992, radius=2, shapeName=dyn_hair_03_r_capsule, center=<4 0 0>, rotation=<rot 90 0 0>}]
 dyn_hair_04_l [{shapeType=Capsule, length=2, radius=1.5, shapeName=dyn_hair_04_l_capsule, center=<3.5 1 -0.5>, rotation=<rot 77.4957 0.0003 -14.2854>}]
 dyn_hair_04_r [{shapeType=Capsule, length=4, radius=2, shapeName=dyn_hair_04_r_capsule, center=<4.7 0 0>, rotation=<rot 90 16.7968 0>}]
 dyn_hair_base_l [{shapeType=Capsule, length=0.505, radius=0.505, shapeName=dyn_hair_base_l_capsule}]
 dyn_hair_base_r [{shapeType=Capsule, length=0.505, radius=0.505, shapeName=dyn_hair_base_r_capsule}]
 head [{shapeType=Capsule, length=5, radius=6.6, shapeName=head_capsule, center=<0 1.3681 0>, rotation=<rot 90 -20.716 0>}]
 neck_01 [{shapeType=Capsule, length=5.9258, radius=5.5487, shapeName=neck_01_capsule, center=<3.3744 0 0>, rotation=<rot 90 0 0>}]
 spine_04 [{shapeType=Capsule, length=13.754, radius=11.9326, shapeName=spine_04_capsule, center=<0.1647 0.6819 -0.2427>}]
 spine_05 [{shapeType=Capsule, length=14.9184, radius=10.8697, shapeName=spine_05_capsule, center=<-0.662 -2.2965 0>}]
 upperarm_l [{shapeType=Capsule, length=17.4602, radius=6.6, shapeName=upperarm_l_capsule, center=<15 0 -1>, rotation=<rot 90 0 0>}]
 upperarm_r [{shapeType=Capsule, length=21.0455, radius=7, shapeName=upperarm_r_capsule, center=<-15 0.5237 -0.3696>, rotation=<rot 90 180 0>}]
}
constraints [
 {bone1Name=head, bone2Name=spine_05, swing1LimitDegrees=45, swing2LimitDegrees=45, twistLimitDegrees=45}
 {bone1Name=upperarm_l, bone2Name=spine_05, swing1LimitDegrees=45, swing2LimitDegrees=45, twistLimitDegrees=45}
 {bone1Name=upperarm_r, bone2Name=spine_05, swing1LimitDegrees=45, swing2LimitDegrees=45, twistLimitDegrees=45}
 {
  bone1Name=spine_05
  bone2Name=spine_04
  swing1LimitDegrees=45
  swing1Motion=Free
  swing2LimitDegrees=45
  swing2Motion=Free
  twistLimitDegrees=45
  twistMotion=Free
 }
 {
  bone1Name=clavicle_l
  bone2Name=spine_05
  swing1LimitDegrees=45
  swing1Motion=Free
  swing2LimitDegrees=45
  swing2Motion=Free
  twistLimitDegrees=45
  twistMotion=Free
 }
 {
  bone1Name=clavicle_r
  bone2Name=spine_05
  swing1LimitDegrees=45
  swing1Motion=Free
  swing2LimitDegrees=45
  swing2Motion=Free
  twistLimitDegrees=45
  twistMotion=Free
 }
 {
  bone1Name=neck_01
  bone2Name=spine_05
  swing1LimitDegrees=45
  swing1Motion=Free
  swing2LimitDegrees=45
  swing2Motion=Free
  twistLimitDegrees=45
  twistMotion=Free
 }
 {
  bone1Name=upperarm_l
  bone2Name=clavicle_l
  swing1LimitDegrees=45
  swing1Motion=Free
  swing2LimitDegrees=45
  swing2Motion=Free
  twistLimitDegrees=45
  twistMotion=Free
 }
 {
  bone1Name=upperarm_r
  bone2Name=clavicle_r
  swing1LimitDegrees=45
  swing1Motion=Free
  swing2LimitDegrees=45
  swing2Motion=Free
  twistLimitDegrees=45
  twistMotion=Free
 }
 {
  bone1Name=head
  bone2Name=neck_01
  swing1LimitDegrees=45
  swing1Motion=Free
  swing2LimitDegrees=45
  swing2Motion=Free
  twistLimitDegrees=45
  twistMotion=Free
 }
 {
  bone1Name=dyn_hair_base_l
  bone2Name=head
  swing1LimitDegrees=45
  swing1Motion=Free
  swing2LimitDegrees=45
  swing2Motion=Free
  twistLimitDegrees=45
  twistMotion=Free
 }
 {
  bone1Name=dyn_hair_base_r
  bone2Name=head
  swing1LimitDegrees=45
  swing1Motion=Free
  swing2LimitDegrees=45
  swing2Motion=Free
  twistLimitDegrees=45
  twistMotion=Free
 }
 {bone1Name=dyn_hair_01_l, bone2Name=dyn_hair_base_l, swing1LimitDegrees=20, swing2LimitDegrees=20, twistLimitDegrees=0}
 {bone1Name=dyn_hair_02_l, bone2Name=dyn_hair_01_l, swing1LimitDegrees=25, swing2LimitDegrees=25, twistLimitDegrees=0}
 {bone1Name=dyn_hair_03_l, bone2Name=dyn_hair_02_l, swing1LimitDegrees=25, swing2LimitDegrees=25, twistLimitDegrees=0}
 {bone1Name=dyn_hair_04_l, bone2Name=dyn_hair_03_l, swing1LimitDegrees=25, swing2LimitDegrees=25, twistLimitDegrees=0}
 {bone1Name=dyn_hair_01_r, bone2Name=dyn_hair_base_r, swing1LimitDegrees=20, swing2LimitDegrees=20, twistLimitDegrees=0}
 {bone1Name=dyn_hair_02_r, bone2Name=dyn_hair_01_r, swing1LimitDegrees=25, swing2LimitDegrees=25, twistLimitDegrees=0}
 {bone1Name=dyn_hair_03_r, bone2Name=dyn_hair_02_r, swing1LimitDegrees=25, swing2LimitDegrees=25, twistLimitDegrees=0}
 {bone1Name=dyn_hair_04_r, bone2Name=dyn_hair_03_r, swing1LimitDegrees=25, swing2LimitDegrees=25, twistLimitDegrees=0}
]
